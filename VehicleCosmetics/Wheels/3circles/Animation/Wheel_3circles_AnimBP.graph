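# Animation Blueprint for the 3circles vehicle cosmetic wheel, playing its looping spinner animation on the wheel skeleton.
# The AnimGraph feeds a single sequence player into the output pose through a local-to-component and back-to-local conversion, with no bone controllers between.
# The EventGraph holds only an unconnected update event and pawn-owner getter, so no variables or gameplay state drive the animation.
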AnimBlueprint /VehicleCosmetics/Wheels/3circles/Animation/Wheel_3circles_AnimBP
target_skeleton=/VehicleCosmetics/Wheels/3circles/Meshes/Wheel_3circles_Skeleton
graphs [
 AnimGraph {nodes [
   n1 = Sequence Player 'Wheel_3circles_Loop_Spinner'() {sequence=&/Wheel_3circles_Loop_Spinner}
   Output Pose(Result=Component To Local(ComponentPose=Local To Component(LocalPose=@n1)))
  ]}
 EventGraph {nodes [
   TryGetPawnOwner()
   Event BlueprintUpdateAnimation()
  ]}
]
variables []
