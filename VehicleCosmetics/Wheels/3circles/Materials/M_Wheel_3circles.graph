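# Masked master material for the 3circles vehicle cosmetic wheel, built on MF_WheelParent_Simple and usable on skeletal meshes.
# It overrides emissive with base color scaled by EmissiveValue, masked by rim blue and inverted vertex red.
# The FXAnim switch adds a panning, sharpened gradient stripe that darkens base color and emissive within rim alpha and blue masks.
# The result passes through the vehicle damage and de-rez transition functions.
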Material /VehicleCosmetics/Wheels/3circles/Materials/M_Wheel_3circles
graph {
 properties {
  bAutomaticallySetUsageInEditor=false
  bUseMaterialAttributes=true
  bUsedWithMeshDeformer=true
  bUsedWithSkeletalMesh=true
  blendMode=BLEND_Masked
  pixelDepthOffsetMode=PDOM_Legacy
  usedShadingModels=Default Lit
 }
 nodes [
  n1 = MaterialFunctionCall() {materialFunction=&MF_WheelParent_Simple}
  n2 = GetMaterialAttributes(None=@n1.MaterialAttributes) {attributeGetTypes [#1, #2]}
  n3 = Constant3Vector()
  n4 = TextureSample(UVs=Panner(Coordinate=TextureCoordinate()) {speedX=0.5}) {texture=/VehicleCosmetics/Wheels/_SharedTextures/FX/T_GradientStripeVertical_M}
  n5 = Multiply(
   A=Multiply(A=Power(Base=@n4.R) {constExponent=10}, B=ComponentMask(None=@n1.RimRGBAMask) {a=true})
   B=ComponentMask(None=@n1.RimRGBAMask) {b=true}
  )
  n6 = StaticSwitchParameter(True=LinearInterpolate(A=@n2.Base Color, B=LinearInterpolate(A=@n2.Base Color, B=@n3), Alpha=@n5), False=@n2.Base Color) {defaultValue=true, group=6_EmissiveProperties, parameterName=FXAnim, sortPriority=0}
  n7 = LinearInterpolate(
   A=@n2.Emissive Color
   B=Multiply(A=ScalarParameter() {defaultValue=1, group=6_EmissiveProperties, parameterName=EmissiveValue, sortPriority=1}, B=@n2.Base Color)
   Alpha=Multiply(A=ComponentMask(None=@n1.RimRGBAMask) {b=true}, B=OneMinus(None=@n1.VertexColorR))
  )
  n8 = Constant() {r=-1}
  n9 = StaticSwitchParameter(True=LinearInterpolate(A=@n7, B=Add(A=@n7, B=@n8), Alpha=@n5), False=@n7) {defaultValue=true, group=6_EmissiveProperties, parameterName=FXAnim, sortPriority=0}
  MaterialFunctionCall(In=MaterialFunctionCall(Material=SetMaterialAttributes!(MaterialAttributes=@n2.MaterialAttributes, Base Color=@n6, Emissive Color=@n9) {attributeSetTypes [#1, #2]}) {materialFunction=&MF_BaseVehicle_Damage}.Result) {materialFunction=&MF_VehicleDeRez_Transition}
  Reroute(None=@n2.Emissive Color)
 ]
}
